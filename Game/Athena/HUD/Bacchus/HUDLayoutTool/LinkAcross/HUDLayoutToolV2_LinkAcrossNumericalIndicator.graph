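# Widget Blueprint for the HUD Layout Tool V2's Link Across numerical indicator, showing an action symbol above an instance counter.
# A fixed-height SizeBox wraps a VerticalBox that splits space between two centered CommonTextBlocks, defaulting to "+" and "0/0".
# Both text blocks are inherited from the native FortniteUI parent class; the Blueprint only supplies layout and styling, as its event graph holds no logic.
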WidgetBlueprint /Game/Athena/HUD/Bacchus/HUDLayoutTool/LinkAcross/HUDLayoutToolV2_LinkAcrossNumericalIndicator
graphs [EventGraph {nodes [
   Event PreConstruct()
   Event Construct()
   Event Tick()
  ]}]
variables []
widgets {
 info {
  inheritedWidgetCount=2
  parentClass=/Script/FortniteUI.HUDLayoutToolV2_LinkAcrossNumericalIndicator
  rootWidgetClass=&SizeBox
  widgetCount=4
 }
 widgets [
  SizeBox_0 {widgetClassPath=&SizeBox, widget=SizeBox_0, properties {bOverride_HeightOverride=true, heightOverride=70}}
  VerticalBox_23 {widgetClassPath=&VerticalBox, parent=SizeBox_0, widget=VerticalBox_23, slot {class=SizeBoxSlot}}
  Text_ActionSymbol {
   widgetClassPath=&CommonTextBlock
   bInherited=true
   parent=VerticalBox_23
   widget=Text_ActionSymbol
   properties {
    justification=Center
    text=+
    textTransformPolicy=ToUpper
    font {
     bForceMonospaced=false
     bMaterialIsStencil=false
     fontMaterial=null
     fontObject=&BurbankSmall
     letterSpacing=0
     monospacedWidth=1
     size=30
     skewAmount=0
     typefaceFontName=Black
     outlineSettings {
      bApplyOutlineToDropShadows=false
      bMiteredCorners=false
      bSeparateFillAlpha=false
      outlineMaterial=null
      outlineSize=0
      outlineColor=<rgba 0 0 0 1>
     }
    }
    renderTransform {angle=0, scale=<1 1>, shear=<0 0>, translation=<0 -10>}
    shadowOffset=<0 0>
   }
   slot {class=VerticalBoxSlot, properties {horizontalAlignment=HAlign_Center, verticalAlignment=VAlign_Top, size {sizeRule=Fill, value=1}}}
  }
  Text_InstanceCounter {
   widgetClassPath=&CommonTextBlock
   bInherited=true
   parent=VerticalBox_23
   widget=Text_InstanceCounter
   properties {
    justification=Center
    text=0/0
    textTransformPolicy=ToUpper
    font {
     bForceMonospaced=false
     bMaterialIsStencil=false
     fontMaterial=null
     fontObject=&BurbankSmall
     letterSpacing=0
     monospacedWidth=1
     size=20
     skewAmount=0
     typefaceFontName=Black
     outlineSettings {
      bApplyOutlineToDropShadows=false
      bMiteredCorners=false
      bSeparateFillAlpha=false
      outlineMaterial=null
      outlineSize=0
      outlineColor=<rgba 0 0 0 1>
     }
    }
    shadowOffset=<0 0>
   }
   slot {class=VerticalBoxSlot, properties {horizontalAlignment=HAlign_Center, verticalAlignment=VAlign_Bottom, size {sizeRule=Fill, value=1}}}
  }
 ]
}
